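# Niagara system for the bathroom bathtub prop that renders a looping GPU field of swirling bubbles as meshes and sprites.
# BubbleLoop continuously spawns dense particles on a plane, culled by a sphere and two crossed boxes.
# A vortex force drives motion with Y force zeroed, while region-masked waveforms pulse sprite size, alpha fades over life and dynamic parameters vary the material.
# The Minimal orbiting-mesh emitter and several extra force modules are disabled.
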NiagaraSystem /Hermes/Props/Bathroom/FX/NS_Bathroom_Bathtub
dependencies {usedDataInterfaces [&NiagaraDataInterfaceCurve], usedDynamicInputs [], usedModules [&SystemState]}
emitters {
 BubbleLoop {
  simTarget=GPUComputeSim
  emitterUpdateScript [
   &EmitterState(Life Cycle Mode=::NewEnumerator1, Loop Duration=1, UseLoopDelay=false)
   &SpawnRate(SpawnRate=2500, Use Spawn Probability=false)
  ]
  particleSpawnScript [
   &InitializeParticle(
    Color=<rgba 0.0529 0.3325 0.3968 1>
    Lifetime Max=1
    Lifetime Mode=::NewEnumerator1
    Mesh Scale Mode=::NewEnumerator1
    Mesh Uniform Scale Max=0.35
    Mesh Uniform Scale Min=0.2
    Sprite Size Mode=::NewEnumerator1
    Uniform Sprite Size Max=12
    Uniform Sprite Size Min=8
    UsePositionOffset=false
    Write Material Random=true
   )
   &ShapeLocation(
    Box / Plane Mode:ENiagara_BoxPlaneMode=::NewEnumerator1
    MaskSpawnGroup=false
    Offset=<0 0 -25>
    Offset Mode=::NewEnumerator0
    Override Seed=false
    Plane Size:Vector2f=<320 320>
    Shape Primitive=::NewEnumerator2
    UseOwnerScale=true
   )
   &InitialMeshOrientation(
    Mesh Axis To Orient=<0 0 1>
    Mesh Orientation Mode=::NewEnumerator2
    Orientation Vector=<0 0 0>
    Rotation=&RandomRangeVector(Minimum=<0 0 0>, Maximum=<0 0 1>)
    Use Rotation Vector=true
   )
   &AddVelocity(
    Module.Use Falloff Distance=false
    Override Seed=false
    Velocity Mode=::NewEnumerator1
    Velocity Speed=&RandomRangeFloat(Minimum=0, Maximum=-25)
   )
   &KillParticlesInVolume(Sphere Radius=135)
   &KillParticlesInVolume(Box Size:Vector3f=<250 150 150>, Kill Shape=::NewEnumerator1) {module=KillParticlesInVolume001}
   &KillParticlesInVolume(Box Size:Vector3f=<150 250 150>, Kill Shape=::NewEnumerator1) {module=KillParticlesInVolume002}
  ]
  particleUpdateScript [
   &CurlNoiseForce(
    Bias Noise=false
    Noise Frequency=&RandomRangeFloat(Minimum=0.25, Maximum=1)
    Noise Strength=25
    Pan Noise=true
    Pan Noise Field=<0 0 1>
   ) {enabled=false}
   &ParticleState()
   &RegionMask(Invert Region=true, Region Radius=180, UseOffset=false, UseRotation=false)
   &ScaleSpriteSize(Uniform Curve Scale=&Multiply_Float(
     A=&Waveform(
      Global Amplitude Scale=2
      "Waveform [1]"=::NewEnumerator0
      "[1] Frequency"=&RandomRangeFloat(Minimum=1, Maximum=2)
      "[1] Amplitude Min/Max"=<1 5>
     )
     B=$Particles.RegionMask.Float
    ))
   &VortexForce(Origin Pull Amount=1, UseOriginOffset=true, Vortex Force Amount=100, Vortex Origin Offset:Vector3f=<0 0 -128>)
   &VortexForce(
    Influence Falloff Radius=1500
    Origin Pull Amount=25
    Use Influence Falloff=true
    UseOriginOffset=true
    Vortex Axis=&Lerp_Vector(
     A=<0 0 1>
     B=<0.25 0 1>
     Alpha=&FloatFromCurve(
      FloatCurve=DataInterface() {propertyValues {Curve {keys [
          {interpMode=RCIM_Cubic, tangentMode=RCTM_User, time=0, value=0}
          {interpMode=RCIM_Cubic, tangentMode=RCTM_User, time=0.5, value=1}
          {tangentMode=RCTM_User, time=1, value=0}
         ]}}}
      CurveIndex=$System.NormalizedLoopAge
     )
    )
    Vortex Force Amount=150
    Vortex Origin Offset:Vector3f=<0 0 -25>
   ) {enabled=false, module=VortexForce001}
   &PointAttractionForce(
    Attraction Radius=1024
    Attraction Strength=0.5
    Attractor Position Offset:Vector3f=<0 0 -128>
    Falloff Exponent=1
    Kill Within Radius=false
    Module.Use Falloff=true
    UseAttractorOffset=true
   ) {enabled=false}
   &PointForce(Force Falloff Distance=500, Force Strength=-50, Linear Force Vector:Vector3f=<0 0 0>, Module.Use Falloff=false, Module.UseLinearForce=false) {enabled=false}
   &SolveForcesAndVelocity(
    Acceleration Limit=5
    Clamp Velocity=true
    Force:Vector3f=&Multiply_Vector(A=$Transient.PhysicsForce, B=<1 0 1>)
    Limit Acceleration=true
    Speed Limit=500
   )
   &ScaleMeshSizeBySpeed(Max Scale Factor=<2 1 1>) {enabled=false}
   &DynamicMaterialParameters(
    Index 0 Param 1=&RandomRangeFloat(Minimum=1, Maximum=3)
    Index 0 Param 2=&RandomRangeFloat(Minimum=3, Maximum=8)
    Index 0 Param 3=1
    Index 0 Param 4=&RandomRangeFloat(Minimum=-0.25, Maximum=-1)
    Index 3 Param 2=&MakeFloatFromVector(Vector=$Particles.Scale, Channel=::NewEnumerator0)
    Index 3 Param 4=&RandomRangeFloat(Minimum=0, Maximum=1)
    Module.Param0WriteEnabled=true
    Module.Param0WriteEnabled3=false
    Module.Param1WriteEnabled=true
    Module.Param1WriteEnabled3=true
    Module.Param2WriteEnabled=true
    Module.Param2WriteEnabled3=false
    Module.Param3WriteEnabled=true
    Module.Param3WriteEnabled3=true
    Write Parameter Index 3=true
   )
   &ScaleColor(
    Scale Alpha=&Multiply_Float(
     A=&FloatFromCurve(FloatCurve=DataInterface() {propertyValues {Curve {keys [{interpMode=RCIM_Cubic, tangentMode=RCTM_User, time=0, value=0}, <at 0.1 1 RCIM_Cubic>, {tangentMode=RCTM_User, time=1, value=0}]}}})
     B=1
    )
    ScaleA=true
    ScaleRGB=true
   )
   &ScaleMeshSize(Scale Factor=&Multiply_VectorByFloat(
     Vector=<1 1 1>
     Float=&FloatFromCurve(
      FloatCurve=DataInterface() {propertyValues {Curve {keys [<at 0 0.25>,<at 0.8 0.925>,<at 1 2>]}}}
      Scale Curve=&Multiply_Float(A=1, B=$Particles.RegionMask.Float)
     )
    )) {enabled=false}
   &CameraOffset(Camera Offset Amount=-5)
   /SetVariables(Particles.float=2) {isSetParameters=true}
  ]
  rendererClasses [&NiagaraMeshRendererProperties, &NiagaraSpriteRendererProperties]
  renderers [
   {rendererClass=&NiagaraMeshRendererProperties, rendererIndex=0}
   {rendererClass=&NiagaraSpriteRendererProperties, rendererIndex=1}
  ]
 }
 Minimal {
  bEnabled=false
  emitterUpdateScript [
   &EmitterState()
   &SpawnBurst_Instantaneous(Spawn Count=8, Use Spawn Probability=false, UseLoopCountLimit=false)
  ]
  particleSpawnScript [
   &InitializeParticle(Color Mode=::NewEnumerator0, Lifetime=5, Mesh Scale Mode=::NewEnumerator3, UsePositionOffset=false, Write Material Random=true)
   &ShapeLocation(
    MaskSpawnGroup=false
    Ring / Disc Distribution Mode:ENiagaraSphereDistributionMode=::NewEnumerator3
    Shape Primitive=::NewEnumerator4
    Uniform Spiral Falloff=0
    UseOwnerScale=true
   )
   &InitialMeshOrientation(Mesh Orientation Mode=::NewEnumerator2, Orientation Vector=<0 1 0>, Rotation=<1 0 1>, Use Rotation Vector=true)
   /SetVariables(Particles.Position=&AddVectorToPosition(Position=$Particles.ShapeLocation.ShapePosition, Vector=$Output.RotateAroundPoint001.OutputPosition)) {enabled=false, isSetParameters=true}
  ]
  particleUpdateScript [
   &ParticleState()
   &RotateAroundPoint(
    Radius:NiagaraFloat=&RandomRangeFloat(Minimum=10, Maximum=100)
    Rotation Center:NiagaraPosition=<1 0 1>
    Rotation Phase:NiagaraFloat=&RandomRangeFloat(Minimum=0, Maximum=1)
    Rotation Rate:NiagaraFloat=&RandomRangeFloat(Minimum=0.95, Maximum=1)
   ) {module=RotateAroundPoint001}
   &UpdateMeshOrientation()
   &SolveForcesAndVelocity(Clamp Velocity=false, Limit Acceleration=false)
  ]
  rendererClasses [&NiagaraMeshRendererProperties]
  renderers [{rendererClass=&NiagaraMeshRendererProperties, rendererIndex=0}]
 }
}
summary {
 systemName=NS_Bathroom_Bathtub
 emitters [
  {bEnabled=false, emitterName=Minimal, rendererClasses [&NiagaraMeshRendererProperties]}
  {
   emitterName=BubbleLoop
   simTarget=GPUComputeSim
   rendererClasses [&NiagaraMeshRendererProperties, &NiagaraSpriteRendererProperties]
  }
 ]
 userVariables [User.BaseColor {type=LinearColor, description="", defaultValue {struct=LinearColor, value=<rgba 1 0.1245 0.1145 1>}}]
}
